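; Fortnite Knock Grenade projectile that sticks on impact or splashes into water, beeps, then explodes after a curve-table delay to launch nearby targets.
; Stopping can attach it to vehicles and stops its spin.
; Explosions launch pawns, AI, physics objects and tagged buildings away using scalable-float velocities with a minimum upward component, respecting knockback immunity.
; Player launches also set fall instigator and broadcast self, teammate or enemy quest verbs.
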(event OnComponentBeginOverlap(CapsuleComponent)
  (params (OverlappedComponent :type Primitive%20Component%20Object%20Reference) (OtherActor :type Actor%20Object%20Reference) (OtherComp :type Primitive%20Component%20Object%20Reference) (OtherBodyIndex :type Integer) (bFromSweep :type Boolean) (SweepResult :type Hit%20Result%20Structure%20%28by%20ref%29))
  (bind (_bblockinghit _binitialoverlap _time _distance _location _impactpoint _normal _impactnormal _physmat _hitactor) (Collision|BreakHitResult SweepResult))
  ;[1 1 1] If we overlap a pawn, stop moving and fall.
  (bind _aspawn (Utilities|Casting|CastToPawn _hitactor))
  (if (Utilities|NotEqual(Object) _aspawn (Game|GetInstigator))
    (Components|Movement|StopMovementImmediately (Variables|Movement|GetProjectileMovementComponent))))

(event OnEnterWaterEvent(WaterInteractionComponent)
  (params (WaterBody :type Fort%20Water%20Body%20Actor%20Object%20Reference) (WaterInteractionComponent :type Fort%20Water%20Interaction%20Component%20Object%20Reference) (bIsFirstBody :type Boolean))
  (bind (_waterplanelocation _waterplanenormal _watersurfaceposition _waterdepth _waterbodyidx) (FX|GetLastWaterSurfaceInfo WaterInteractionComponent :!class /Script/FortniteGame.FortWaterInteractionComponent))
  ;[1 1 1] Handle Entering Water, Play splash FX and Sounds then trigger the beeping and the explosion event
  (Utilities|FlowControl|DoOnce false
    (:Completed
      (Variables|BPrjGrenade|SetHitLocation (Transformation|GetActorLocation))
      (Variables|BPrjGrenade|SetHitNormal "0, 0,1.000000")
      (bind _returnvalue (Effects|Components|ParticleSystem|SpawnFXAtLocation "/Game/Weapons/FORT_Generic_Effects/Impacts/Water/NS_Impact_Water_Athena/NS_Impact_Water_Beeg_Apollo.NS_Impact_Water_Beeg_Apollo" _watersurfaceposition (Math|Rotator|MakeRotfromX _waterplanenormal)))
      (Effects|Components|ParticleSystem|SetFloatParameter _returnvalue "WaterDepth" _waterdepth)
      (Effects|Components|ParticleSystem|SetFloatParameter _returnvalue "WaterBodyIndex" (Math|Conversions|ToFloat(Integer) _waterbodyidx))
      (Audio|PlaySoundatLocation "/Game/Sounds/Fort_GamePlay_Sounds/Explosives/Grenade_Projectile_Bounce_Water_Splash_NoLoop_Cue.Grenade_Projectile_Bounce_Water_Splash_NoLoop_Cue" (Transformation|GetActorLocation))
      (CallFunction|ForceExplode))))

(event EventOnTouched
  (params (OtherActor :type Actor%20Object%20Reference) (OtherComp :type Primitive%20Component%20Object%20Reference) (HitResult :type Hit%20Result%20Structure%20%28by%20ref%29) (bIsOverlap :type Boolean))
  (bind (_bblockinghit _binitialoverlap _time _distance _location _impactpoint _normal _impactnormal _physmat _hitactor _hitcomponent) (Collision|BreakHitResult HitResult))
  ;[1 1 1] If we touch water do nothing
  (Utilities|Casting|CastToFortWaterBodyActor OtherActor
    (:then
      (Components|Movement|StopMovementImmediately (Variables|Movement|GetProjectileMovementComponent))
      (Variables|Projectile|SetbStopSimulatingOnHit true))
    (:CastFailed
      (if (not (Components|ComponentHasTag _hitcomponent "NoStick"))
        (Variables|Projectile|SetbStopSimulatingOnHit true)))))

(event EventBeginPlay
  ;[1 1 1] Begin Play
  (Networking|CanExecuteCosmeticEvents
    (:True
      (Audio|Components|Audio|SetSound (Variables|B_Prj_Grenade|GetGrenadeFuse_AudioComponent) (Variables|BPrjGrenade|GetCue_GrenadeFuseSound)))))

(event OnProjectileBounce(ProjectileMovementComponent)
  (params (ImpactResult :type Hit%20Result%20Structure%20%28by%20ref%29) (ImpactVelocity :type Vector%20%28by%20ref%29 :default "0, 0, 0"))
  (Utilities|Time|SetTimerbyEvent (EventDispatchers|CreateEvent :!event DoExplosion) 0.4))

(event Custom|Stop_Rotation
  ;[1 1 1] Stop Rotating Movement
  (Components|Movement|SetUpdatedComponent (Variables|Default|GetRotatingMovement) 0))

(event Custom|ForceExplode
  ;[0.7116 0.838 1] OnStop - wait then do explosion
  (Audio|Components|Audio|Play (Variables|Default|GetBeepTimer) 0.0)
  (Audio|Components|Audio|FadeOut (Variables|Default|GetAudio_InAirLoop) 0.1 0.0 "Linear" :!class /Script/Engine.AudioComponent)
  (Audio|PlaySoundatLocation "/Game/Sounds/Fort_GamePlay_Sounds/Explosives/Grenade_Impulse_Stick_Cue.Grenade_Impulse_Stick_Cue" (Transformation|GetActorLocation))
  (bind _outxy (DataTable|EvaluateCurveTableRow "/Game/Athena/Balance/DataTables/AthenaGameData.AthenaGameData" "Default.KnockGrenade.OnHitExplodeDelay")
    (:RowFound
      (Utilities|FlowControl|Delay _outxy)
      (Delivery|DoExplosion))))

(event EventOnStop
  (params (Hit_bBlockingHit :type Boolean) (Hit_bInitialOverlap :type Boolean) (Hit_Time :type Float%20%28single-precision%29) (Hit_Distance :type Float%20%28single-precision%29) (Hit_Location :type Vector :default "0, 0, 0") (Hit_ImpactPoint :type Vector :default "0, 0, 0") (Hit_Normal :type Vector :default "0, 0, 0") (Hit_ImpactNormal :type Vector :default "0, 0, 0") (Hit_PhysMat :type Physical%20Material%20Object%20Reference) (Hit_HitActor :type Actor%20Object%20Reference) (Hit_HitComponent :type Primitive%20Component%20Object%20Reference) (Hit_HitBoneName :type Name :default "None") (Hit_BoneName :type Name :default "None") (Hit_HitItem :type Integer) (Hit_ElementIndex :type Integer) (Hit_FaceIndex :type Integer) (Hit_TraceStart :type Vector :default "0, 0, 0") (Hit_TraceEnd :type Vector :default "0, 0, 0"))
  ;[0.7116 0.838 1] OnStop - wait then do explosion
  (Variables|BPrjGrenade|SetHitLocation Hit_Location)
  (Variables|BPrjGrenade|SetHitNormal Hit_Normal)
  (Utilities|FlowControl|Sequence
    (:then_0
      (bind _asfort_vehicle_interface (Utilities|Casting|CastToFortVehicleInterface Hit_HitActor)
        (:then
          (Transformation|AttachActorToActor (Vehicle|GetActor _asfort_vehicle_interface) "None" "KeepWorld" "KeepWorld" "KeepWorld")
          (Variables|Default|SetStuckVehicle true))
        (:CastFailed
          _))
      (CallFunction|StopRotation)
      (Audio|Components|Audio|Play (Variables|Default|GetBeepTimer) 0.0)
      (Audio|Components|Audio|FadeOut (Variables|Default|GetAudio_InAirLoop) 0.1 0.0 "Linear" :!class /Script/Engine.AudioComponent)
      (Audio|PlaySoundatLocation "/Game/Sounds/Fort_GamePlay_Sounds/Explosives/Grenade_Impulse_Stick_Cue.Grenade_Impulse_Stick_Cue" (Transformation|GetActorLocation))
      (bind _outxy (DataTable|EvaluateCurveTableRow "/Game/Athena/Balance/DataTables/AthenaGameData.AthenaGameData" "Default.KnockGrenade.OnHitExplodeDelay")
        (:RowFound
          (Utilities|FlowControl|Delay _outxy)
          (Delivery|DoExplosion))))
    (:then_1
      ;;[1 0.8419 0.6198] Dust Impact
      (Effects|Components|ParticleSystem|SpawnFXAtLocation "/Game/Athena/Items/Consumables/KnockGrenade/FX/NS_Athena_KnockGrenade_Dust.NS_Athena_KnockGrenade_Dust" (Variables|BPrjGrenade|GetHitLocation) (Math|Conversions|RotationFromXVector (Variables|BPrjGrenade|GetHitNormal))))))

(event EventOnExploded
  (params (HitActors :type Array%20of%20Actor%20Object%20References) (HitResults :type Array%20of%20Hit%20Result%20Structures))
  (bind _target (Variables|BPrjGrenade|GetTarget))
  ;[1 1 1] OnExploded - Emitter, Sound, Camera Shake
  (Audio|Components|Audio|FadeOut (Variables|Default|GetBeepTimer) 4.0 0.0 "Linear" :!class /Script/Engine.AudioComponent)
  (CallFunction|SoundIndicatorExplosion)
  (Utilities|FlowControl|Sequence
    (:then_0
      (for _array_element HitActors
        (bind _asfort_pawn (Utilities|Casting|CastToFortPawn _array_element)
          (:then
            (Variables|BPrjGrenade|SetTarget _asfort_pawn)
            (bind _outxy (DataTable|EvaluateCurveTableRow "/Game/Athena/Balance/DataTables/AthenaGameData.AthenaGameData" "Default.KnockGrenade.FriendlyFire")
              (:RowFound
                (if (not (> _outxy 0.0))
                  ;[1 1 1] Perform a team check. Same team?
                  (if (and (Utilities|NotEqual(Object) (Variables|Actor|GetInstigator) (Variables|BPrjGrenade|GetTarget)) (Team|OnSameTeam (Variables|BPrjGrenade|GetTarget) (Variables|Actor|GetInstigator)))
                    _
                    (else
                      _)))
                (bind _asfort_player_pawn (Utilities|Casting|CastToFortPlayerPawn (Variables|BPrjGrenade|GetTarget)))
                (if (Vehicle|IsinVehicle _asfort_player_pawn)
                  _
                  (elif (GameplayTags|HasMatchingGameplayTag (Variables|BPrjGrenade|GetTarget) "(TagName=\"Gameplay.Status.KnockbackImmunity\")")
                    _
                    (else
                      ;[0 0.4531 0.2084] Launch Player
                      (Utilities|FlowControl|Sequence
                        (:then_0
                          (CallFunction|HandlePawnDetachRC)
                          (Variables|/Script/FortniteGame.FortPawn|SetbIgnoreNextFallingDamage false (Variables|BPrjGrenade|GetTarget))
                          (Components|Movement|StopMovementImmediately (Variables|/Script/Engine.Character|GetCharacterMovement _target))
                          (Character|LaunchCharacter _target (CallFunction|GetLaunchVel _target (Variables|Default|GetRow_KnockbackVel) (Variables|Default|GetRow_KnockbackAddedZ)) true true)
                          (Networking|SwitchHasAuthority
                            (:Authority
                              (bind _asfort_player_pawn_athena (Utilities|Casting|CastToFortPlayerPawnAthena _target))
                              (Class|FortPlayerPawnAthena|SetFallInstigator _asfort_player_pawn_athena (Game|GetInstigatorController) (Variables|Tags|GetDefaultTags)))))
                        (:then_1
                          (Effects|SpawnFXSystemAttached "/Game/Athena/Items/Consumables/KnockGrenade/FX/NS_Athena_KnockGrenade_Player_Trails.NS_Athena_KnockGrenade_Player_Trails" (Variables|/Script/Engine.Character|GetMesh (Variables|BPrjGrenade|GetTarget))))
                        (:then_2
                          ;[0.3816 0.979 1] Quest Custom Verbs
                          ; Set up separate verbs for self and other players
                          (if (== (Variables|Actor|GetInstigator) (Variables|BPrjGrenade|GetTarget))
                            ;;[1 1 1 0.5] Self
                            (CustomVerb|BroadcastCustomVerbMessage (CustomVerb|MakeCustomVerbMessage (Variables|Actor|GetInstigator) (Variables|Default|GetImpulseSelfQuestVerbTag)))
                            (else
                              (if (Team|OnSameTeam (Variables|Actor|GetInstigator) (Variables|BPrjGrenade|GetTarget))
                                ;;[1 1 1 0.5] Teammates
                                (CustomVerb|BroadcastCustomVerbMessage (CustomVerb|MakeCustomVerbMessage (Variables|Actor|GetInstigator) (Variables|Default|GetImpulseTeammateQuestVerbTag)))
                                (else
                                  ;;[1 1 1 0.5] Enemies
                                  (CustomVerb|BroadcastCustomVerbMessage (CustomVerb|MakeCustomVerbMessage (Variables|Actor|GetInstigator) (Variables|Default|GetImpulseEnemyQuestVerbTag)))))))))))))
              (:RowNotFound
                _)))
          (:CastFailed
            (CallFunction|LaunchNonPawnActors _array_element))))
      ;;[1 0.8789 0.2904] FortAIPawn
      (for _array_element_2 HitActors
        (bind _asfort_aipawn (Utilities|Casting|CastToFortAIPawn _array_element_2))
        (if (not (GameplayTags|HasMatchingGameplayTag _asfort_aipawn "(TagName=\"Gameplay.Status.KnockbackImmunity\")"))
          (Variables|Default|SetFortAIPawn _asfort_aipawn)
          (CallFunction|LaunchFortAIPawn))))
    (:then_1
      (Networking|CanExecuteCosmeticEvents
        (:True
          (if (Water|IsinWater (Variables|Water|GetWaterInteractionComponent))
            (CallFunction|WaterFXExplosion)
            (else
              (Effects|Components|ParticleSystem|SpawnFXAtLocation (Variables|Default|GetFX%20Explosion) (select (Variables|Projectile|GetbStopSimulatingOnHit) (Variables|BPrjGrenade|GetHitLocation) (Transformation|GetActorLocation)) (Math|Conversions|RotationFromXVector (Variables|BPrjGrenade|GetHitNormal)))))
          (if (Pawn|IsLocallyControlled (Game|GetInstigator))
            (Audio|PlaySoundatLocation (Variables|Sounds|GetCue_CloseSound) (Transformation|GetActorLocation) "0, 0, 0" 0.6) ; Play Close Explosion
            (Audio|Components|Audio|FadeOut (Variables|B_Prj_Grenade|GetGrenadeFuse_AudioComponent) 0.2 0.0 "Linear" :!class /Script/Engine.AudioComponent) ; Fade Out Fuse Sound
            (else
              (Audio|PlaySoundatLocation (Variables|Sounds|GetCue_DistantSound) (Transformation|GetActorLocation) "0, 0, 0" 0.4) ; Play Distant Explosion
            )))))))

(event Custom|LaunchFortAIPawn
  (bind _fortaipawn (Variables|Default|GetFortAIPawn))
  ;[0.967 1 0.5015] Launch FortAIPawn
  (Utilities|FlowControl|Sequence
    (:then_0
      (Utilities|IsValid _fortaipawn
        (:Is%20Valid
          (Components|Movement|StopMovementImmediately (Variables|/Script/Engine.Character|GetCharacterMovement _fortaipawn))
          (Character|LaunchCharacter _fortaipawn (CallFunction|GetLaunchVel (Variables|Default|GetFortAIPawn) (Variables|Default|GetRow_KnockbackVel) (Variables|Default|GetRow_KnockbackAddedZ)) true true))))
    (:then_1
      (Effects|SpawnFXSystemAttached "/Game/Athena/Items/Consumables/KnockGrenade/FX/NS_Athena_KnockGrenade_Player_Trails.NS_Athena_KnockGrenade_Player_Trails" (Variables|/Script/Engine.Character|GetMesh (Variables|Default|GetFortAIPawn))))))

(event Custom|Water%20FX%20Explosion
  (bind _waterinteractioncomponent (Variables|Water|GetWaterInteractionComponent))
  (bind (_waterplanelocation _waterplanenormal _watersurfaceposition _waterdepth _waterbodyidx) (FX|GetLastWaterSurfaceInfo _waterinteractioncomponent :!class /Script/FortniteGame.FortWaterInteractionComponent))
  ;[1 1 1] OnExploded - Emitter, Sound, Camera Shake
  ;;[1 1 1] Water FX Explosion
  (bind _returnvalue (Effects|Components|ParticleSystem|SpawnFXAtLocation (Variables|Default|GetFX%20Water%20Explosion) (Variables|BPrjGrenade|GetHitLocation) (Math|Rotator|MakeRotfromX _waterplanenormal)))
  (Effects|Components|ParticleSystem|SetFloatParameter _returnvalue "WaterDepth" _waterdepth)
  (Effects|Components|ParticleSystem|SetVectorParameter _returnvalue "WaterPlaneLocation" _waterplanelocation)
  (Effects|Components|ParticleSystem|SetVectorParameter _returnvalue "WaterPlaneNormal" _waterplanenormal)
  (Effects|Components|ParticleSystem|SetFloatParameter _returnvalue "WaterBodyIndex" (Math|Conversions|ToFloat(Integer) _waterbodyidx)))

(event Custom|LaunchNonPawnActors
  (params (Actor :type Actor%20Object%20Reference))
  ;[0.3228 0.9901 1] If the actor is a FortPhysicsPawnObject, apply the impulse directly to it, otherwise try to launch it as a building.
  (bind _asfort_physics_pawn_object (Utilities|Casting|CastToFortPhysicsPawnObject Actor)
    (:then
      (Utilities|IsValid _asfort_physics_pawn_object
        (:Is%20Valid
          (bind (_initialparameters_initialvelocity _initialparameters_overridemassinkg _initialparameters_overrideangulardampening _initialparameters_overridelineardampening _initialparameters_playerforcemultiplier) (Utilities|Struct|BreakPhysicsPawnObjectInitialParameters (Variables|/Script/FortniteGame.FortPhysicsPawnObject|GetInitialParameters_InitialVelocity _asfort_physics_pawn_object)))
          (PhysicsObject|ApplyImpulse _asfort_physics_pawn_object self (* (* (CallFunction|GetLaunchVel _asfort_physics_pawn_object (Variables|Default|GetRow_KnockbackVel) (Variables|Default|GetRow_KnockbackAddedZ)) (ScalableFloat|GetValueatLevel (Variables|Default|GetRow_PhysicsPawnObjectMult))) (* (/ _initialparameters_playerforcemultiplier (Math|Float|Max(Float) (Variables|Default|GetPhysicsPawnObjectForceModifier) 1.0)) (ScalableFloat|GetValueatLevel (Variables|Default|GetRow_PhysicsPawnObjectMult))))))))
    (:CastFailed
      (CallFunction|LaunchBuldingActors Actor))))

(event Custom|LaunchBuldingActors
  (params (Actor :type Actor%20Object%20Reference))
  ;[1 1 1] If the actor has the building actor tag and also a physics component launch it.
  (bind _returnvalue (GameplayTags|HasMatchingGameplayTag(Message) Actor (Variables|Default|GetTag_BuildingPhysics)))
  (if _returnvalue
    (Utilities|IsValid (Actor|GetComponentbyClass Actor "/Script/FortniteGame.FortPhysicsObjectComponent")
      (:Is%20Valid
        (Utilities|Casting|CastToPrimitiveComponent (Variables|/Script/Engine.Actor|GetRootComponent Actor))
        (FortPhysicsSimulationLibrary|AddPhysicsImpulseatLocation Actor (* (CallFunction|GetLaunchVel Actor (Variables|Default|GetRow_KnockbackVel) (Variables|Default|GetRow_KnockbackAddedZ)) (ScalableFloat|GetValueatLevel (Variables|Default|GetRow_BuildingImpulseMult))) (Transformation|GetActorLocation Actor) true "None" false 0 "(MaxSimSizeAllowed=Large,bRejectEmbeddedInLandscape=False)")))))

(fn GetAffiliationsToShowFor
  (return (Utilities|Array|MakeArray "Hostile" "Neutral" "Friendly")))

(fn GetLaunchVel
  (params (Actor :type Actor%20Object%20Reference) (Magnitude :type Scalable%20Float%20Structure%20%28by%20ref%29) (AddedZ :type Scalable%20Float%20Structure%20%28by%20ref%29))
  ;[1 1 1] Calculate how far to launch the actor based on rows.
  (return (Math|Vector|MakeVector (.x (* (Math|Vector|Normalize (- (Transformation|GetActorLocation Actor) (Variables|BPrjGrenade|GetHitLocation))) (ScalableFloat|GetValueatLevel Magnitude))) (.y (* (Math|Vector|Normalize (- (Transformation|GetActorLocation Actor) (Variables|BPrjGrenade|GetHitLocation))) (ScalableFloat|GetValueatLevel Magnitude))) (Math|Float|Max(Float) (+ (.z (* (Math|Vector|Normalize (- (Transformation|GetActorLocation Actor) (Variables|BPrjGrenade|GetHitLocation))) (ScalableFloat|GetValueatLevel Magnitude))) (ScalableFloat|GetValueatLevel AddedZ)) 5e+02))))

(fn HandlePawnDetachRC
  (bind _asfort_player_pawn (Utilities|Casting|CastToFortPlayerPawn (Variables|BPrjGrenade|GetTarget)))
  (RemoteControlPawn|ForceDetachfromRemoteControlledPawn _asfort_player_pawn))

(fn ShouldShowSoundIndicator
  (params (PlayerController :type Fort%20Player%20Controller%20Object%20Reference))
  (return true))

(fn SoundIndicatorExplosion
  (bind _returnvalue (Transformation|GetActorLocation))
  (bind _fortsoundindicatorexplosion (Variables|Default|GetFortSoundIndicatorExplosion))
  (SoundIndicator|StopSoundIndicatorLooping (Variables|Default|GetFortSoundIndicator))
  (if (Pawn|IsLocallyControlled (Game|GetInstigator))
    (SoundIndicator|ShowSoundIndicator _fortsoundindicatorexplosion _returnvalue (Variables|Sounds|GetCue_CloseSound) 1.0 0)
    (else
      (SoundIndicator|ShowSoundIndicator _fortsoundindicatorexplosion _returnvalue (Variables|Sounds|GetCue_DistantSound) 1.0 0))))

(fn ConstructionScript
  (Networking|CanExecuteCosmeticEvents
    (:True
      (Components|Movement|SetUpdatedComponent (Variables|Default|GetRotatingMovement) (Variables|B_Prj_Grenade|GetMesh)))))
